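# Niagara system for the Lasso emote's ring bling, a brief sparkle glint.
# A single sprite emitter, Ring_Sparkle, burst-spawns one particle at the system location with a random rotation and an overbright white color.
# Over the particle's life, curves fade color and alpha in to a midpoint peak and back out, while sprite size holds steady until halfway, then shrinks to zero.
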NiagaraSystem /BRCosmetics/Effects/Fort_Effects/Emotes/Lasso/NS_Emote_Lasso_RingBling
dependencies {usedDataInterfaces [&NiagaraDataInterfaceCurve], usedDynamicInputs [], usedModules [&SystemState]}
emitters {Ring_Sparkle {
  emitterUpdateScript [
   &EmitterState(Life Cycle Mode=::NewEnumerator1, Loop Behavior=::NewEnumerator1, Loop Duration=1, UseLoopDelay=false)
   &SpawnBurst_Instantaneous(Spawn Count=1, Use Spawn Probability=false)
  ]
  particleSpawnScript [
   /Niagara/Modules/Spawn/Initialization/InitializeParticle(
    Color=<rgba 10 10 10 1>
    Module.Write Color=true
    Module.Write Lifetime=true
    Module.Write Mass=true
    Module.Write Material Random=true
    Module.Write Position=true
    Module.Write Scale=false
    Module.Write SpriteRotation=true
    Module.Write SpriteSize=true
    Sprite Rotation=&UniformRangedFloat(Minimum=0, Maximum=360)
    Sprite Size=&Vector2DFromFloat(Value=35)
   )
   &SystemLocation(MaskSpawnGroup=false)
  ]
  particleUpdateScript [
   &ScaleColor(
    Scale Alpha=&FloatFromCurve(FloatCurve=DataInterface() {propertyValues {Curve {keys [<at 0 0>,<at 0.5016 0.9767>,<at 1 0>]}}})
    Scale RGB=&VectorFromFloat(Value=&FloatFromCurve(FloatCurve=DataInterface() {propertyValues {Curve {keys [<at 0 0>,<at 0.5017 1.0042>,<at 1 0>]}}}))
    ScaleA=true
    ScaleRGB=true
   )
   &ParticleState()
   &SolveForcesAndVelocity(Clamp Velocity=false, Limit Acceleration=false)
   &ScaleSpriteSize(Scale Factor=&Vector2DFromFloat(Value=&FloatFromCurve(FloatCurve=DataInterface() {propertyValues {Curve {keys [
          <at 0 1 RCIM_Cubic>
          {arriveTangent=-0.01, interpMode=RCIM_Cubic, leaveTangent=-0.01, time=0.335, value=1}
          {arriveTangent=-0.0228, interpMode=RCIM_Cubic, leaveTangent=-2.1462, tangentMode=RCTM_Break, time=0.5002, value=0.995}
          {arriveTangent=-1.9908, interpMode=RCIM_Cubic, leaveTangent=-1.9908, time=0.6669, value=0.2877}
          <at 1 0>
         ]}}})))
  ]
  rendererClasses [&NiagaraSpriteRendererProperties]
  renderers [{rendererClass=&NiagaraSpriteRendererProperties}]
 }}
summary {
 systemName=NS_Emote_Lasso_RingBling
 emitters [{emitterName=Ring_Sparkle, rendererClasses [&NiagaraSpriteRendererProperties]}]
 userVariables []
}
